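# Animation Blueprint for the FloppyNib quadruped cosmetic companion, adding a laser-chasing emote over a pose retargeted from another mesh.
# Each update copies laser, crouch, pounce and swipe state from the owner's emote component, picking random left or right timed swipes.
# Cached crouch or pounce poses feed a bone-layered swipe, then spine CCDIK, head look-at and tail trail.
# When not following the laser, it outputs the plain retargeted pose.
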AnimBlueprint /CosmeticCompanions/Assets/Quadruped/FloppyNib/Meshes/FloppyNib_Quadruped_AnimBP
target_skeleton=&/FloppyNib_Quadruped_Skeleton
graphs [
 AnimGraph {nodes [
   Copy Pose From Mesh() {bCopyCurves=true, bUseAttachedParent=true}
   Retarget Pose From Mesh!() -> (Pose:Pose Link Structure) {is=title AnimGraphNode_RetargetPoseFromMesh, iKRetargeterAsset=&/RTG_FloppyNib}
   n1 = Layered blend per bone(
    BasePose=Use cached pose 'BasePose'()
    BlendPoses_0=Sequence Player!(Sequence:Anim Sequence Base Object Reference) -> (Pose) {bLoopAnimation=false}
   ) {layerSetup [{branchFilters [
       {boneName=shoulder_fr_r}
       {boneName=elbow_fr_r}
       {boneName=wrist_fr_r}
       {boneName=ball_fr_r}
       {boneName=toe_fr_r}
       {boneName=spine_02}
       {boneName=spine_03}
       {boneName=neck_01}
       {boneName=neck_02}
       {boneName=head}
       {boneName=spine_01}
       {boneName=pelvis}
       {boneName=IK_Foot_Front_R}
      ]}]}
   n2 = Blend Poses by bool(BlendPose_0=@n1, BlendPose_1=Use cached pose 'BasePose'(), BlendTime_1=0.2, bActiveValue=Get bSwipeActive() -> (bSwipeActive:bool)) {childUpateMode=ResetChildOnActivate}
   n3 = CCDIK!(
    ComponentPose:Component Space Pose Link Structure=Local To Component(LocalPose=Slot 'DefaultSlot'(Source=@n2))
    Alpha:float=0.5
    EffectorLocation:Vector="0,0,0"
   ) -> (Pose:Component Space Pose Link Structure) {
    is=title AnimGraphNode_CCDIK
    bEnableRotationLimit=true
    effectorLocationSpace=BCS_WorldSpace
    maxIterations=5
    precision=2
    rootBone {boneName=spine_01}
    rotationLimitPerJoints [15,30,45,30]
    tipBone {boneName=neck_01}
   }
   n4 = Trail controller - Bone: tail_05(
    ComponentPose=Look At - Bone: head!(ComponentPose=@n3, Alpha=0.75, LookAtLocation:Vector="100,0,0") -> (Pose) {lookAtClamp=70, boneToModify {boneName=head}, lookAt_Axis {axis {y=0}}, lookAtTarget=~}
    Alpha=1
   ) {
    chainLength=5
    stretchLimit=0.1
    rotationLimits [{limitMax=<180 180 180>}, {limitMax=<180 180 180>}, {limitMax=<180 180 180>}, {limitMax=<180 180 180>}, {limitMax=<180 180 180>}]
    rotationOffsets [<0 0 0>, <0 0 0>, <0 0 0>, <0 0 0>, <0 0 0>]
    trailBone {boneName=tail_05}
    alphaBoolBlend=~
   }
   Output Pose(Result=Blend Poses by bool(
     BlendPose_0=Component To Local(ComponentPose=@n4)
     BlendPose_1=Use cached pose 'Cached RTG'()
     bActiveValue=Get bIsFollowingLaser() -> (bIsFollowingLaser:bool)
    ))
   n5 = "Sequence Player (additive) 'Flourcut_Crouch_ADDRT'"() {bLoopAnimation=false, sequence=/CosmeticCompanions/Assets/Quadruped/FloppyNib/Animation/Emotes/FlobbyNib_LaserEmote/Flourcut_Crouch_ADDRT}
   n6 = Layered blend per bone(
    BasePose=Apply Additive(Base=Use cached pose 'Cached RTG'(), Additive=@n5)
    BlendPoses_0=Sequence Player 'FloppyNib_CrouchLoop_RT'() {sequence=/CosmeticCompanions/Assets/Quadruped/FloppyNib/Animation/Emotes/FlobbyNib_LaserEmote/FloppyNib_CrouchLoop_RT}
   ) {layerSetup [{branchFilters [
       {boneName=shoulder_fr_r}
       {boneName=elbow_fr_r}
       {boneName=wrist_fr_r}
       {boneName=ball_fr_r}
       {boneName=toe_fr_r}
       {boneName=spine_02}
       {boneName=spine_03}
       {boneName=neck_01}
       {boneName=neck_02}
       {boneName=head}
       {boneName=spine_01}
       {boneName=pelvis}
       {boneName=IK_Foot_Front_R}
      ]}]}
   n7 = Blend Poses by bool(
    BlendPose_0=@n6
    BlendPose_1=Use cached pose 'Cached RTG'()
    BlendTime_0=0.25
    BlendTime_1=0.25
    bActiveValue=Get bShouldCrouch() -> (bShouldCrouch:bool)
   )
   n8 = Sequence Player 'Flourcut_PounceRT2'() {bLoopAnimation=false, sequence=/CosmeticCompanions/Assets/Quadruped/FloppyNib/Animation/Emotes/FlobbyNib_LaserEmote/Flourcut_PounceRT2}
   n9 = Get bShouldPounce() -> (bShouldPounce:bool)
   Save cached pose 'BasePose'(Pose=Modify Curve(
     SourcePose=Blend Poses by bool(BlendPose_0=@n8, BlendPose_1=@n7, BlendTime_0=0.25, BlendTime_1=0.5, bActiveValue=@n9) {childUpateMode=ResetChildOnActivate}
     CurveValues_0=1
     Alpha=1
    ) {curveNames [UseRootSlopeOrientation], curveValues [0]}) {cacheName=BasePose}
   Save cached pose 'Cached RTG'(Pose=Retarget Pose From Mesh() {iKRetargeterAsset=&RTG_FloppyNib}) {cacheName=Cached RTG}
  ]}
 Emote_LaserFollow {nodes [
   n1 = Emote_LaserFollow!() -> (then, DeltaTime:double=~) {is=title K2Node_FunctionEntry}
   n2 = GetComponentByClass!(
    self:Actor Object Reference=GetOwningActor()
    ComponentClass:Actor Component Class Reference=/CosmeticCompanions/Assets/Quadruped/FloppyNib/Animation/Emotes/FlobbyNib_LaserEmote/FloppyNib_Emote_Component.FloppyNib_Emote_Component_C
   ) -> (ReturnValue:Floppy Nib Emote Component Object Reference)
   > n3 = Branch(Condition=IsValid(Object=@n2))
   > Set bIsFollowingLaser(bIsFollowingLaser:bool=Get bIsFollowingLaser!(self:Floppy Nib Emote Component Object Reference=@n2) -> (bIsFollowingLaser)) -> (Output_Get:bool)
   > Set bShouldCrouch(bShouldCrouch:bool=Get bShouldCrouch!(self:Floppy Nib Emote Component Object Reference=@n2) -> (bShouldCrouch)) -> (Output_Get:bool)
   > Set bShouldPounce(bShouldPounce:bool=Get bShouldPounce!(self:Floppy Nib Emote Component Object Reference=@n2) -> (bShouldPounce)) -> (Output_Get:bool)
   > Set LaserTargetLocation(LaserTargetLocation:Vector=Get LaserTargetLocation!(self:Floppy Nib Emote Component Object Reference=@n2) -> (LaserTargetLocation:Vector)) -> (Output_Get:Vector)
   > Set SwipeTimer(SwipeTimer:double=float - float(A=Get SwipeTimer() -> (SwipeTimer:double), B=@n1)) -> (Output_Get:double)
   > n4 = Branch(Condition=Get bShouldSwipe!(self:Floppy Nib Emote Component Object Reference=@n2) -> (bShouldSwipe:bool))
   > Set SwipeAnim(SwipeAnim:Anim Sequence Object Reference=Select(
     Option 0:Anim Sequence Object Reference=/CosmeticCompanions/Assets/Quadruped/FloppyNib/Animation/Emotes/FlobbyNib_LaserEmote/FloppyNib_Emote_SwipeLeft
     Option 1:Anim Sequence Object Reference=/CosmeticCompanions/Assets/Quadruped/FloppyNib/Animation/Emotes/FlobbyNib_LaserEmote/FloppyNib_Emote_SwipeRight
     Index:int=RandomIntegerInRange!(Min:int=0, Max:int=1) -> (ReturnValue:int=0)
    ) -> (ReturnValue:Anim Sequence Object Reference)) -> (Output_Get:Anim Sequence Object Reference)
   > Set bSwipeActive(bSwipeActive:bool=true) -> (Output_Get:bool)
   > Set SwipeTimer(SwipeTimer=0.4)
   > Set bShouldSwipe!(execute, bShouldSwipe:bool=false, self:Floppy Nib Emote Component Object Reference=@n2) -> (then, Output_Get:bool)
   n5 = Get SwipeTimer()
   Branch(execute=@n4.else, Condition=AND Boolean(A="float <= float"(A=@n5) -> (ReturnValue=~), B=Get bSwipeActive()))
   > Set bSwipeActive(bSwipeActive=false)
   Set bIsFollowingLaser(execute=@n3.else, bIsFollowingLaser=false)
   n6 = Branch(Condition="float > float"(B=2.36))
   n7 = GetComponentByClass(
    self=GetOwningActor()
    ComponentClass=&FloppyNib_Emote_Component_C
   )
   > Set bHasGoal!(execute, bHasGoal:bool=true, self:Floppy Nib Emote Component Object Reference=@n7) -> (then, Output_Get:bool)
   Set bHasGoal(execute=@n6.else, bHasGoal=false, self=@n7)
   Vector Length(A=GetVelocity!(self:Actor Object Reference=GetOwningActor()) -> (ReturnValue:Vector))
   GetAnimInstance(self=GetOwningComponent())
   n8 = GetOwningActor()
   Set Distance(Distance:double="Distance (Vector)"!(V1:Vector=Get Actor Location(self=@n8), V2:Vector=Get LaserTargetLocation!() -> (LaserTargetLocation)) -> (ReturnValue:double)) -> (Output_Get:double)
  ]}
 EventGraph {nodes [
   n1 = Event BlueprintUpdateAnimation()
   > Emote_LaserFollow!(execute, self, DeltaTime:double=@n1.DeltaTimeX) -> (then) {is=title K2Node_CallFunction}
   Event BlueprintInitializeAnimation()
  ]}
]
variables [
 bIsFollowingLaser {type=boolean}
 bShouldCrouch {type=boolean}
 bShouldPounce {type=boolean}
 bSwipeActive {type=boolean}
 ClampedLaserLocation {type=Vector}
 LaserTargetLocation {type=Vector}
 SwipeAnim {type=AnimSequence, default=null}
 SwipeTimer {type=number}
]
